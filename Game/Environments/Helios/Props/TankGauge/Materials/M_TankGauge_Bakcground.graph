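# Default Lit material for the background of the Helios tank gauge prop, with diffuse, normal and mask texture samples.
# A glow term cubes the diffuse texture's RGB, then tints it by the orange GlowColor parameter and scales it by GlowStrength from the MPC_TankGauge parameter collection.
# A soft sphere mask on local position confines that glow around the local origin.
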Material /Game/Environments/Helios/Props/TankGauge/Materials/M_TankGauge_Bakcground
graph {
 properties {bAutomaticallySetUsageInEditor=false, pixelDepthOffsetMode=PDOM_Legacy, usedShadingModels=Default Lit}
 nodes [
  n1 = TextureSample() {texture=/Game/Environments/Helios/Props/TankGauge/Textures/T_TankGauge_A_D}
  n2 = VectorParameter() {parameterName=GlowColor, defaultValue=<rgba 1 0.4815 0 1>}
  n3 = CollectionParameter() {collection=&/MPC_TankGauge, parameterName=GlowStrength}
  Multiply(
   A=Multiply(A=Power(Base=@n1.RGB) {constExponent=3}, B=Multiply(A=@n2.RGB, B=@n3) {constB=25})
   B=SphereMask(A=LocalPosition().XYZ, B=Constant3Vector()) {attenuationRadius=60, hardnessPercent=50}
  )
  TextureSample() {samplerType=SAMPLERTYPE_Normal, texture=/Game/Environments/Helios/Props/TankGauge/Textures/T_TankGauge_A_N}
  TextureSample() {samplerType=SAMPLERTYPE_Masks, texture=/Game/Environments/Helios/Props/TankGauge/Textures/T_TankGauge_A_S}
 ]
}
